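# MF_3DHB_POST_RimLight is a post material function in the 3DHB shader MasterLib that produces a colored rim-light term for emissive.
# It takes a tangent-space normal, dots it with a direction parameter, inverts the result and raises it to a fresnel power.
# The rim color tints this falloff for the AddToEmissive output.
# The fresnel-like effect uses a parameter direction instead of the camera vector, with no normalization in the function.
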MaterialFunction /BRCosmetics/Tools/3DHB/Shader/MasterLib/PostFunctions/MF_3DHB_POST_RimLight
graph {nodes [
  n1 = OneMinus(None=DotProduct(
    A=FunctionInput() {bUsePreviewValueAsDefault=true, inputName="NormalIn(TS)", previewValue=<0 0 1 1>}
    B=VectorParameter() {group=POST_RimLight, parameterName="RimLight_Direction(TS)", defaultValue=<rgba 0 0 1 1>}.RGB
   ))
  n2 = ScalarParameter() {defaultValue=3, group=POST_RimLight, parameterName=RimLight_FresnelPower}
  FunctionOutput(None=Multiply(A=Power(Base=@n1, Exp=@n2), B=VectorParameter() {group=POST_RimLight, parameterName=RimLight_Color, defaultValue=<rgba 0 0.2496 3 1>}.RGB)) {outputName=AddToEmissive}
 ]}
